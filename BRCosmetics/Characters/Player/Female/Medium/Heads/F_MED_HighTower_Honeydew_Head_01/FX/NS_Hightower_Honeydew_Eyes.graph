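# Niagara eye-glow effect for the Hightower Honeydew female head cosmetic, drawing two golden flares with sprite and light renderers.
# A single emitter bursts one particle in each of two spawn groups, and both particles persist past their lifetime.
# Each update samples the user skeletal mesh, using spawn group to pick one of two mirrored offsets, one per eye.
# A small camera offset is then applied to the particles.
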NiagaraSystem /BRCosmetics/Characters/Player/Female/Medium/Heads/F_MED_HighTower_Honeydew_Head_01/FX/NS_Hightower_Honeydew_Eyes
dependencies {usedDataInterfaces [], usedDynamicInputs [], usedModules [&SystemState]}
emitters {Flare {
  emitterUpdateScript [
   &EmitterState(
    Inactive Response=::NewEnumerator1
    Life Cycle Mode=::NewEnumerator1
    Loop Behavior=::NewEnumerator1
    Loop Duration Mode=::NewEnumerator1
    UseLoopDelay=false
   )
   &SpawnBurst_Instantaneous(Spawn Count=1, Use Spawn Probability=false)
   &SpawnBurst_Instantaneous(Spawn Count=1, Spawn Group=1, Use Spawn Probability=false) {module=SpawnBurst_Instantaneous001}
  ]
  particleSpawnScript [/Niagara/Modules/Spawn/Initialization/InitializeParticle(
    Color=<rgba 1 0.6733 0.02 0.4>
    Mesh Scale=<1.3 1.3 1.3>
    Module.Write Color=true
    Module.Write Lifetime=true
    Module.Write Mass=true
    Module.Write Material Random=true
    Module.Write Position=false
    Module.Write Scale=false
    Module.Write SpriteRotation=false
    Module.Write SpriteSize=true
    Position=&SimulationPosition()
    Sprite Size=&Vector2DFromFloat(Value=16.5)
   )]
  particleUpdateScript [
   &ParticleState(Kill Particles When Lifetime Has Elapsed=false, Let Infinitely Lived Particles Die When Emitter Deactivates=true)
   &SkeletalMeshLocation(
    Bone Sampling Mode=::NewEnumerator4
    MaskSpawnGroup=false
    Sampled Position Offset=&Lerp_Vector(A=<3.2 7 3.2>, B=<3.2 7 -3.2>, Alpha=&MakeFloatFromInt(Int=&SpawnGroup()))
    Skeletal Mesh=$User.SkeletalMesh
   )
   &CameraOffset(Camera Offset Amount=4)
  ]
  rendererClasses [&NiagaraSpriteRendererProperties, &NiagaraLightRendererProperties]
  renderers [
   {rendererClass=&NiagaraSpriteRendererProperties}
   {rendererClass=&NiagaraLightRendererProperties, rendererIndex=1}
  ]
 }}
summary {
 systemName=NS_Hightower_Honeydew_Eyes
 emitters [{emitterName=Flare, rendererClasses [&NiagaraSpriteRendererProperties, &NiagaraLightRendererProperties]}]
 userVariables [
  User.Right2ImpactVelocity {type=Vector3f, description="", defaultValue {struct=Vector3f, value=<0 0 0>}}
  User.SkeletalMesh {
   type=NiagaraDataInterfaceSkeletalMesh
   defaultValue {
    struct=NiagaraExt_VariableValue_DataInterface
    value {dataInterface=&self:NiagaraDataInterfaceSkeletalMesh_0}
   }
  }
 ]
}
